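; Gameplay ability for a compact car delivery quest, crediting the player for bringing a qualifying car into a tagged POI.
; On a vehicle-entered event it checks the vehicle's ability system for the car tag and, only if the player is outside the POI, waits for the POI tag to be added.
; It then rechecks the car and sends the quest stat event before ending; a vehicle-exit event also ends the ability.
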(event Ability|EventActivateAbilityFromEvent
  (params (EventData :type Gameplay%20Event%20Data%20Structure%20%28by%20ref%29))
  (bind _owningplayerpawn (Variables|GAVars|Data|GetOwningPlayerPawn))
  (|CallParentFunction EventData)
  (switch Utilities|FlowControl|Switch|SwitchonGameplayTag (Utilities|Struct|BreakGameplayEventData EventData)
    (:Event.Vehicle.Entered
      ;[1 1 1] Check Tags
      (if (GameplayTags|TagContainer|CompareTagAssetInterfacetoOtherTags (Variables|Default|GetT_Car) (Vehicle|GetVehicleAbilitySystemComponent (Vehicle|GetVehicleInterface _owningplayerpawn)))
        (if (not (GameplayTags|TagContainer|CompareTagAssetInterfacetoOtherTags (Variables|Default|GetPOI%20Tag) _owningplayerpawn))
          (Ability|Tasks|WaitGameplayTagAdd (Variables|Default|GetPOI%20Tag) 0 true
            (:Added
              (if (GameplayTags|TagContainer|CompareTagAssetInterfacetoOtherTags (Variables|Default|GetT_Car) (Vehicle|GetVehicleAbilitySystemComponent (Vehicle|GetVehicleInterface (Variables|GAVars|Data|GetOwningPlayerPawn))))
                (Quest|SendComplexCustomStatEvent (Variables|GAVars|Data|GetOwningPlayerQuestManager) 0 (Variables|Default|GetT_Quest) "(GameplayTags=)" 1)
                (else
                  _))
              (Ability|EndAbility))))
        (else
          _))
      (Ability|EndAbility))
    (:Event.Vehicle.Exit
      (Ability|EndAbility))))
